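# Niagara skydiving contrail cosmetic with pink flames, sparks, stars, bokeh sparkles and glowing, lit horseshoe meshes.
# A shared socket reader positions all five emitters, with ribbons and sprites trailing against the skydive velocity.
# The system's Dive value drives spawn rates, speeds and horseshoe tilt, which also uses a smoothed interpolated value mirrored per side.
# Front-end modifiers, time-of-day tints and camera-distance scaling adapt the look to context.
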NiagaraSystem /BRCosmetics/Athena/Effects/Particles/Skydiving/NS_Contrail_ForestJean/FX/NS_Contrail_ForestJean
dependencies {
 usedDataInterfaces [
  &NiagaraDataInterfaceCurve
  &NiagaraDataInterfaceCamera
  &NiagaraDataInterfaceColorCurve
  &NiagaraDataInterfaceArrayFloat
  &NiagaraDataInterfaceArrayColor
  &NiagaraDataInterfaceArrayInt32
  &NiagaraDataInterfaceSocketReader
 ]
 usedDynamicInputs [
  &FN_FrontEndVector3Modifier
  &Lerp_Vector
  &DotProduct
 ]
 usedModules [
  &self:SystemSpawnScript.SystemScriptSource.SystemScriptGraph.NiagaraNodeAssignment_6.SetVariables_#1
  &SystemState
  &self:SystemSpawnScript.SystemScriptSource.SystemScriptGraph.NiagaraNodeAssignment_0.SetVariables_#2
  &self:SystemSpawnScript.SystemScriptSource.SystemScriptGraph.NiagaraNodeAssignment_1.SetVariables_#3
  &self:SystemSpawnScript.SystemScriptSource.SystemScriptGraph.NiagaraNodeAssignment_4.SetVariables_#4
  &InterpolateOverTime
 ]
}
emitters {
 Bokeh_and_Sparkles {
  emitterUpdateScript [
   &EmitterState()
   &SpawnRate(
    SpawnRate=&FloatFromCurve(
     FloatCurve=DataInterface() {propertyValues {Curve {
        defaultValue=1
        keys [{interpMode=RCIM_Cubic, tangentMode=RCTM_User, time=0.2, value=0}, {interpMode=RCIM_Cubic, tangentMode=RCTM_User, time=1, value=1}]
       }}}
     CurveIndex=$System.Dive
     Scale Curve=100
    )
    Use Spawn Probability=false
   )
  ]
  particleSpawnScript [
   &PartitionParticles(Number Of Partitions=4)
   &InitializeParticle(
    Color=&SelectLinearColorFromArray(
     Array Sampling Mode=::NewEnumerator0
     Direct Array Index=?
     Color Selection Array=DataInterface() {propertyValues {ColorData [<rgba 0.8638 0.1826 1 1>, <rgba 1.2 0.1983 0.5119 1>, <rgba 0.8 0.1113 0.3922 1>, <rgba 1.5 0.5913 0.5087 1>]}}
     Random Seed=0
    )
    Lifetime=&Multiply_Float(
     A=&RandomRangeFloat(Minimum=0.4, Maximum=0.8)
     B=&FN_FrontEndFloatModifier(Gameplay=1, Front End=0.75)
    )
    Sprite Rotation Mode=::NewEnumerator0
    Sprite Size Mode=::NewEnumerator1
    Uniform Sprite Size Max=100
    Uniform Sprite Size Min=80
    UsePositionOffset=false
    Write Material Random=true
   )
   &SocketLocation(
    Interpolate Socket Sampling=true
    SamplePosOffset=false
    Socket Index=4
    Socket Reader=$System.Socket Reader
    Socket Sampling Mode=::NewEnumerator3
    VelocitySpeedLimit=false
   )
   &ShapeLocation(
    Cylinder Height=200
    Cylinder Height Midpoint:NiagaraFloat=&FN_FrontEndFloatModifier(Gameplay=0.6, Front End=0.8)
    Cylinder Radius=&FN_FrontEndFloatModifier(Gameplay=45, Front End=60)
    MaskSpawnGroup=false
    Override Seed=false
    Rotation Axis=&NormalizeVector(
     Vector To Normalize=&TransformVector(Vector=$System.SkydiveVelocity, Source Space=::World, Destination Space=::Local)
     Safe Normalize=true
     Safe Fallback Threshold=0
     Safe Fallback Vector=<0 0 0.001>
    )
    Rotation Mode=::NewEnumerator1
    Shape Primitive=::NewEnumerator1
    Surface Only=false
    UseOwnerScale=true
   )
   &AddVelocity(
    Cone Angle=10
    Cone Axis=&Multiply_VectorByFloat(
     Vector=&NormalizeVector(
      Vector To Normalize=&TransformVector(Vector=$System.SkydiveVelocity, Source Space=::World, Destination Space=::Local)
      Safe Normalize=true
      Safe Fallback Threshold=0
      Safe Fallback Vector=<0 0 0.001>
     )
     Float=-1
    )
    Module.UseConeFalloff=false
    Module.UseVelDistribution=false
    Override Seed=false
    Velocity Mode=::NewEnumerator2
    Velocity Speed=&Multiply_Float(
     A=&Lerp_Float(
      A=&RandomRangeFloat(Minimum=80, Maximum=100)
      B=&RandomRangeFloat(Minimum=175, Maximum=250)
      Alpha=$System.Dive
     )
     B=1
    )
   )
   &FN_SetSpawnGroup(Number of Spawn Groups=8)
   /SetVariables(Particles.VisibilityTag=&SelectIntFromArray(
     Direct Array Index=&SpawnGroup()
     Int Selection Array=DataInterface() {propertyValues {IntData [0,0,0,0,1,1,2,2]}}
    )) {isSetParameters=true, module=SetVariables_#5}
   &CameraOffset(Camera Offset Amount=&SelectFloatFromArray(
     Direct Array Index=&SpawnGroup()
     Float Selection Array=DataInterface() {propertyValues {FloatData [-30,-15,10,20,10,-10,10,-10]}}
    ))
  ]
  particleUpdateScript [
   &ParticleState()
   &ScaleColor(
    Scale Alpha=&FloatFromCurve(
     FloatCurve=DataInterface() {propertyValues {Curve {
        defaultValue=1
        keys [
         {arriveTangent=3.9041, interpMode=RCIM_Cubic, leaveTangent=3.9041, tangentMode=RCTM_User, time=0, value=0}
         {interpMode=RCIM_Cubic, tangentMode=RCTM_User, time=0.5, value=1}
         {arriveTangent=-4.7411, tangentMode=RCTM_User, time=1, value=0}
        ]
       }}}
     Scale Curve=&Multiply_Float(
      A=&Multiply_Float(
       A=&RandomRangeFloat(Minimum=0.1, Maximum=0.15)
       B=&FNTimeOfDayScalar(Night Scale=0.75, Morning Scale=1, Evening Scale=1)
      )
      B=&FN_FrontEndFloatModifier(Gameplay=1, Front End=0.6)
     )
    )
    ScaleA=true
    ScaleRGB=true
   )
   &ScaleSpriteSize(Scale Factor=&Vector2DFromFloat(Value=&FloatFromCurve(
      FloatCurve=DataInterface() {propertyValues {Curve {
         defaultValue=0.8
         keys [
          {interpMode=RCIM_Cubic, tangentMode=RCTM_User, time=0, value=0.65}
          {interpMode=RCIM_Cubic, tangentMode=RCTM_User, time=0.5, value=1}
          {tangentMode=RCTM_User, time=1, value=0.65}
         ]
        }}}
      Scale Curve=/Game/Effects/Niagara/Modules/Math/FN_CameraDistance(
       0Dist:NiagaraFloat=50
       1Dist:NiagaraFloat=100
       Camera Query:NiagaraDataInterfaceCamera=DataInterface() {propertyValues {bRequireCurrentFrameData=false}}
       Position:NiagaraPosition=$Particles.Position
      )
     ))) {module=ScaleSpriteSize001}
   &Drag(Drag=&RandomRangeFloat(Minimum=0, Maximum=1), Use Linear Drag=true, Use Rotational Drag=false)
   &SolveForcesAndVelocity(Clamp Velocity=false, Limit Acceleration=false)
   /SetVariables(Particles.Glow_Color:LinearColor=&MakeLinearColorFromVectorAndFloat(
     "Vector (RGB)"=&MakeVectorFromLinearColorRGB(Color=&SelectLinearColorFromArray(
       Array Sampling Mode=::NewEnumerator0
       Direct Array Index=?
       Color Selection Array=DataInterface() {propertyValues {ColorData [<rgba 1 0.4 0.9969 1>, <rgba 1.2 0.24 0.5405 1>, <rgba 1 0.3304 0.3304 1>]}}
       Random Seed=0
      ))
     "Float (Alpha)"=&FloatFromCurve(
      FloatCurve=DataInterface() {propertyValues {Curve {
         defaultValue=0
         keys [
          {arriveTangent=5.7606, interpMode=RCIM_Cubic, leaveTangent=5.7606, tangentMode=RCTM_User, time=0, value=0}
          {interpMode=RCIM_Cubic, tangentMode=RCTM_User, time=0.3, value=1}
          {tangentMode=RCTM_User, time=1, value=0}
         ]
        }}}
      Scale Curve=&Multiply_Float(
       A=&Multiply_Float(
        A=&RandomRangeFloat(Minimum=0.05, Maximum=0.1)
        B=&FNTimeOfDayScalar(Night Scale=1, Morning Scale=1, Evening Scale=1)
       )
       B=&FN_FrontEndFloatModifier(Gameplay=1, Front End=0.6)
      )
     )
    )) {isSetParameters=true, module=SetVariables_#6}
   /SetVariables(Particles.Glow_Size:Vector2f=&Multiply_Vector2D_ByFloat(Float=&RandomRangeFloat(Minimum=0.25, Maximum=0.4), Vector2D=$Particles.SpriteSize)) {isSetParameters=true, module=SetVariables_#7}
   &FN_Sprite_Attributes(
    Color=&MultiplyLinearColor(
     Linear Color=$Particles.Color
     Scale Factor=&MakeVector4FromVectorAndFloat("Vector (RGB)"=&VectorFromFloat(Value=15), "Float (Alpha)"=1)
    )
    Create Color Attribute=true
    Create SpriteSize Attribute=true
    SpriteSize=&Multiply_Vector2D_ByFloat(
     Float=&FloatFromCurve(
      FloatCurve=DataInterface() {propertyValues {Curve {
         defaultValue=1
         keys [
          {interpMode=RCIM_Cubic, tangentMode=RCTM_User, time=0, value=0}
          {interpMode=RCIM_Cubic, tangentMode=RCTM_User, time=0.5, value=1}
          {tangentMode=RCTM_User, time=1, value=0}
         ]
        }}}
      Scale Curve=&RandomRangeFloat(Minimum=0.15, Maximum=0.25)
     )
     Vector2D=$Particles.SpriteSize
    )
    Write to Module Namespace=true
   )
   &FN_Sprite_Attributes(
    Color=&Multiply_LinearColorByFloat(Linear Color=$Particles.Color, Float=1)
    Create Color Attribute=true
    Create SpriteSize Attribute=true
    SpriteSize=&Multiply_Vector2D_ByFloat(Float=1.5, Vector2D=$Particles.FN_Sprite_Attributes.SpriteSize)
    Write to Module Namespace=true
   ) {module=FN_Sprite_Attributes001}
  ]
  rendererClasses [&NiagaraSpriteRendererProperties]
  renderers [
   {rendererClass=&NiagaraSpriteRendererProperties}
   {rendererClass=&NiagaraSpriteRendererProperties, rendererIndex=1}
   {rendererClass=&NiagaraSpriteRendererProperties, rendererIndex=2}
   {rendererClass=&NiagaraSpriteRendererProperties, rendererIndex=3}
  ]
 }
 Flame_Ribbon {
  emitterSpawnScript [/SetVariables() {isSetParameters=true}]
  emitterUpdateScript [
   &EmitterState()
   &SpawnRate(
    SpawnRate=&RemapRange(
     Input Value To Remap=&VectorLength(Vector=$System.SkydiveVelocity)
     Input Min=1200
     Input Max=4000
     Output Min=100
     Output Max=200
    )
    Use Spawn Probability=false
   )
   /SetVariables(Emitter.UVOffset:NiagaraFloat=&Add_Float(
     A=$Emitter.UVOffset
     B=&Multiply_Float(A=$Engine.DeltaTime, B=&Lerp_Float(A=-1, B=-1.5, Alpha=$System.Dive))
    )) {isSetParameters=true}
  ]
  particleSpawnScript [
   &FN_SetSpawnGroup(Number of Spawn Groups=2)
   /SetVariables(Particles.RibbonID=/Niagara/DynamicInputs/Meshes/MakeRibbonIDFromSpawnGroup()) {isSetParameters=true}
   &InitializeParticle(
    Lifetime=&Lerp_Float(A=0.175, B=0.15, Alpha=$System.Dive)
    Ribbon Width=&Multiply_Float(
     A=50
     B=&RemapRange(
      Input Value To Remap=&VectorLength(Vector=$System.SkydiveVelocity)
      Input Min=1200
      Input Max=4000
      Output Min=1
      Output Max=1.05
     )
    )
    Ribbon Width Mode=::NewEnumerator1
    UsePositionOffset=false
    Write Material Random=true
   )
   &SocketLocation(
    SamplePosOffset=true
    Sampled Position Offset=&Multiply_VectorByFloat(
     Vector=&NormalizeVector(
      Vector To Normalize=&TransformVector(Vector=$System.SkydiveVelocity, Source Space=::World, Destination Space=::Local)
      Safe Normalize=true
      Safe Fallback Threshold=0
      Safe Fallback Vector=<0 0 0.001>
     )
     Float=&FN_FrontEndFloatModifier(
      Gameplay=&RemapRange(
       Input Value To Remap=&VectorLength(Vector=$System.SkydiveVelocity)
       Input Min=1200
       Input Max=6000
       Output Min=45
       Output Max=100
      )
      Front End=&Lerp_Float(A=30, B=40, Alpha=$System.Dive)
     )
    )
    Sampled Position Offset Coordinate Space:ENiagaraSkelMeshTransforms=::NewEnumerator2
    Socket Index=&SpawnGroup()
    Socket Reader=$System.Socket Reader
    Socket Sampling Mode=::NewEnumerator3
    VelocitySpeedLimit=false
   )
   /SetVariables(Particles.RibbonWidth=&Multiply_Float(
     A=$Particles.RibbonWidth
     B=&Lerp_Float(
      A=0.9
      B=0.7
      Alpha=&MakeFloatFromInt(Int=/Niagara/DynamicInputs/Divide/Divide_Int(A:NiagaraInt32=&SpawnGroup(), B:NiagaraInt32=2))
     )
    )) {isSetParameters=true}
  ]
  particleUpdateScript [
   &ParticleState()
   /SetVariables(Particles.Velocity=&Multiply_VectorByFloat(
     Vector=&TransformVector(Vector=$System.SkydiveVelocity, Source Space=::World, Destination Space=::Local, Apply Scale=true)
     Float=-1
    )) {isSetParameters=true}
   &SolveForcesAndVelocity(Clamp Velocity=false, Limit Acceleration=false)
   &DynamicMaterialParameters(
    Index 0 Param 1=&Add_Float(A=$Emitter.UVOffset, B=&MaskFloatBySpawnGroup(Masked Float=0, Passthrough Float=0.5))
    Index 0 Param 2=$System.Dive
    Index 0 Param 3=-5
    Index 1 Param 1=&MakeFloatFromVector(Vector=$System.SkydiveVelocity, Channel=::NewEnumerator0)
    Index 1 Param 2=&MakeFloatFromVector(Vector=$System.SkydiveVelocity, Channel=::NewEnumerator1)
    Index 1 Param 3=&MakeFloatFromVector(Vector=$System.SkydiveVelocity, Channel=::NewEnumerator2)
    Module.Index0Float0Write=true
    Module.Index0Float1Write=true
    Module.Index0Float2Write=true
    Module.Index0Float3Write=true
    Module.Index1Float0Write=true
    Module.Index1Float1Write=true
    Module.Index1Float2Write=true
    Module.Index1Float3Write=true
    Write Parameter Index 1=::NewEnumerator1
   )
   &ScaleRibbonWidth(Ribbon Width Scale=&FloatFromCurve(FloatCurve=DataInterface() {propertyValues {Curve {
        defaultValue=1.4
        keys [{interpMode=RCIM_Cubic, tangentMode=RCTM_User, time=0.3, value=1}, {arriveTangent=-3.2159, tangentMode=RCTM_User, time=1, value=0}]
       }}}))
   &ScaleColor(
    Scale RGB=&MakeVectorFromLinearColorRGB(Color=&FNTimeOfDayColor(
      Night Color=<rgba 0.8075 0.83 0.95 1>
      Morning Color=<rgba 0.95 0.57 0.8872 1>
      Day Color=<rgba 1 1 1 1>
      Evening Color=<rgba 0.95 0.665 0.9029 1>
     ))
    ScaleA=true
    ScaleRGB=true
   )
  ]
  rendererClasses [&NiagaraRibbonRendererProperties]
  renderers [{rendererClass=&NiagaraRibbonRendererProperties}]
 }
 Flames_Sprites {
  emitterUpdateScript [
   &EmitterState(Life Cycle Mode=::NewEnumerator1, Loop Behavior=::NewEnumerator1, Loop Duration Mode=::NewEnumerator1, UseLoopDelay=false)
   &SpawnRate(
    SpawnRate=&Lerp_Float(A=30, B=&FN_FrontEndFloatModifier(Gameplay=50, Front End=75), Alpha=$System.Dive)
    Use Spawn Probability=false
   )
  ]
  particleSpawnScript [
   &FN_SetSpawnGroup(Number of Spawn Groups=2)
   /SetVariables(Particles.ParticleRandom:NiagaraFloat=&RandomRangeFloat(Minimum=0, Maximum=1)) {isSetParameters=true}
   &InitializeParticle(
    Color=&LerpLinearColors(StartColor=<rgba 0.7031 0.2353 0.6543 1>, EndColor=<rgba 1 1 1 1>, LerpFactor=$System.Dive)
    Lifetime=&Multiply_Float(
     A=&RandomRangeFloat(Minimum=0.8, Maximum=1)
     B=&FN_FrontEndFloatModifier(Gameplay=1, Front End=0.75)
    )
    Sprite Rotation Mode=::NewEnumerator2
    Sprite Size Mode=::NewEnumerator3
    Uniform Sprite Size=24
    UsePositionOffset=true
    Write Material Random=true
   )
   &DynamicMaterialParameters(
    Index 0 Param 1=&RandomRangeFloat(Minimum=0, Maximum=1)
    Module.Index0Float0Write=true
    Module.Index0Float1Write=true
    Module.Index0Float2Write=true
    Module.Index0Float3Write=true
   )
  ]
  particleUpdateScript [
   &ParticleState()
   &ScaleSpriteSize(Scale Factor=&Vector2DFromFloat(Value=&FloatFromCurve(
      FloatCurve=DataInterface() {propertyValues {Curve {keys [
          {arriveTangent=2.889, interpMode=RCIM_Cubic, leaveTangent=2.889, tangentMode=RCTM_User, time=0, value=0.4}
          {arriveTangent=0.0217, interpMode=RCIM_Cubic, leaveTangent=0.0217, tangentMode=RCTM_User, time=1, value=2}
         ]}}}
      Scale Curve=&Multiply_Float(
       A=&FN_FrontEndFloatModifier(Gameplay=1, Front End=1)
       B=&FN_CameraDistance(
        0Dist:NiagaraFloat=50
        1Dist:NiagaraFloat=70
        Camera Query:NiagaraDataInterfaceCamera=DataInterface() {propertyValues {bRequireCurrentFrameData=false}}
        Position:NiagaraPosition=$Particles.Position
       )
      )
     )))
   &ScaleColor(
    Scale Alpha=&FloatFromCurve(
     FloatCurve=DataInterface() {propertyValues {Curve {
        defaultValue=1
        keys [
         {arriveTangent=-0.0403, interpMode=RCIM_Cubic, leaveTangent=-0.0403, tangentMode=RCTM_User, time=0.15, value=1}
         {arriveTangent=-0.199, tangentMode=RCTM_User, time=1, value=0}
        ]
       }}}
     Scale Curve=&Multiply_Float(
      A=&RandomRangeFloat(Minimum=0.3, Maximum=0.6)
      B=&FN_FrontEndFloatModifier(Gameplay=0.4, Front End=0.25)
     )
    )
    Scale RGB=&MakeVectorFromLinearColorRGB(Color=&ColorFromCurve(ColorCurve=DataInterface() {propertyValues {
        AlphaCurve {keys [<at 0.2926 1 RCIM_Cubic>]}
        BlueCurve {keys [
          <at 0 2.7658>
          {arriveTangent=2.3844, interpMode=RCIM_Cubic, leaveTangent=2.3844, tangentMode=RCTM_User, time=0.229, value=3.2189}
          <at 0.485 3>
          <at 0.7421 0.29 RCIM_Cubic>
         ]}
        GreenCurve {keys [
          <at 0 0.0901>
          {arriveTangent=2.3844, interpMode=RCIM_Cubic, leaveTangent=2.3844, tangentMode=RCTM_User, time=0.229, value=0.5}
          <at 0.485 0.0987>
          <at 0.7421 0 RCIM_Cubic>
         ]}
        RedCurve {keys [
          <at 0 2.5998>
          {arriveTangent=2.3844, interpMode=RCIM_Cubic, leaveTangent=2.3844, tangentMode=RCTM_User, time=0.229, value=5}
          <at 0.485 2.7984>
          <at 0.7421 0.5 RCIM_Cubic>
         ]}
       }}))
    ScaleA=true
    ScaleRGB=true
   )
   &SocketLocation(
    SamplePosOffset=true
    Socket Index=&SpawnGroup()
    Socket Reader=$System.Socket Reader
    Socket Sampling Mode=::NewEnumerator3
    VelocitySpeedLimit=false
   ) {module=SocketLocation001}
   /SetVariables(Particles.PositionControl:NiagaraFloat=&FloatFromCurve(
     FloatCurve=DataInterface() {propertyValues {Curve {
        defaultValue=0
        keys [
         {interpMode=RCIM_Cubic, tangentMode=RCTM_User, time=0, value=0}
         {arriveTangent=1.3513, interpMode=RCIM_Cubic, leaveTangent=1.3513, tangentMode=RCTM_User, time=1, value=1}
        ]
       }}}
     Scale Curve=&Multiply_Float(
      A=&Multiply_Float(
       A=&RandomRangeFloat(Minimum=-100, Maximum=-120)
       B=&FN_FrontEndFloatModifier(Gameplay=1, Front End=0.5)
      )
      B=&FloatFromCurve(
       FloatCurve=DataInterface() {propertyValues {Curve {keys [{arriveTangent=1.6154, time=0, value=1}, {arriveTangent=2.6178, leaveTangent=2.6178, time=1, value=2}]}}}
       CurveIndex=$System.Dive
      )
     )
    )) {isSetParameters=true}
   /SetVariables(Particles.Position=&AddVectorToPosition(
     Position=$Particles.Position
     Vector=&Multiply_VectorByFloat(
      Vector=&Add_Vector(
       A=&NormalizeVector(
        Vector To Normalize=&TransformVector(Vector=$System.SkydiveVelocity, Source Space=::World, Destination Space=::Simulation)
        Safe Normalize=true
        Safe Fallback Threshold=0
        Safe Fallback Vector=<0 0 0.001>
       )
       B=&RandomVector(Vector Scale=&Lerp_Float(A=0, B=&FN_FrontEndFloatModifier(Gameplay=0.02, Front End=0.02), Alpha=$System.Dive))
      )
      Float=$Particles.PositionControl
     )
    )) {isSetParameters=true}
   &CameraOffset(Camera Offset Amount=20)
  ]
  rendererClasses [&NiagaraSpriteRendererProperties]
  renderers [{rendererClass=&NiagaraSpriteRendererProperties}]
 }
 Light_and_Glow_and_Horse_Shoe {
  emitterUpdateScript [
   &EmitterState(
    Inactive Response=::NewEnumerator1
    Life Cycle Mode=::NewEnumerator1
    Loop Behavior=::NewEnumerator1
    Loop Duration Mode=::NewEnumerator1
    UseLoopDelay=false
   )
   &SpawnBurst_Instantaneous(Spawn Count=2, Use Spawn Probability=false) {module=SpawnBurst_Instantaneous001}
  ]
  particleSpawnScript [
   &FN_SetSpawnGroup(Number of Spawn Groups=2)
   &InitializeParticle(
    Color=<rgba 1 0 0.5 1>
    Lifetime=1.5
    Mesh Scale Mode=::NewEnumerator3
    Mesh Uniform Scale=1.25
    Sprite Size Mode=::NewEnumerator3
    Uniform Sprite Size=32
    UsePositionOffset=false
    Write Material Random=true
   ) {module=InitializeParticle001}
   /SetVariables() {isSetParameters=true}
   &InitialMeshOrientation(
    Mesh Axis To Orient=<0 0 1>
    Mesh Orientation Mode=::NewEnumerator2
    Orientation Coordinate Space=::Local
    Orientation Vector=&MaskVectorBySpawnGroup(Masked Vector=<0 -1 0>, Passthrough Vector=<0 1 0>)
    Rotation=<0 0 0.25>
    Use Rotation Vector=true
   )
   &InitialMeshOrientation(Mesh Orientation Mode=::NewEnumerator7, Rotation=<0 0 -0.05>, Use Rotation Vector=true) {module=InitialMeshOrientation001}
  ]
  particleUpdateScript [
   &ParticleState(Kill Particles When Lifetime Has Elapsed=false)
   &ScaleMeshSize(Scale Factor=&VectorFromFloat(Value=&FloatFromCurve(
      FloatCurve=DataInterface() {propertyValues {Curve {
         defaultValue=0
         keys [{interpMode=RCIM_Cubic, tangentMode=RCTM_User, time=0, value=0.4}, {interpMode=RCIM_Cubic, tangentMode=RCTM_User, time=0.7, value=1}]
        }}}
      CurveIndex=$Emitter.Age
     )))
   &SocketLocation(
    SamplePosOffset=true
    Socket Index=$Particles.SpawnGroup
    Socket Reader=$System.Socket Reader
    Socket Sampling Mode=::NewEnumerator3
    VelocitySpeedLimit=false
   )
   /SetVariables(Particles.MeshOrientation=&MultiplyQuaternion(
     Quaternion A=$Particles.MeshOrientation
     Quaternion B=&MakeQuaternion(
      Quaternion From=2
      Axis=<0 0 0>
      Angle=0
      Y=&FloatFromCurve(
       FloatCurve=DataInterface() {propertyValues {Curve {
          defaultValue=0
          keys [{interpMode=RCIM_Cubic, tangentMode=RCTM_User, time=0, value=0.4}, {interpMode=RCIM_Cubic, tangentMode=RCTM_User, time=0.7, value=1}]
         }}}
       CurveIndex=$Emitter.Age
      )
      From=?
      To=?
      Angle Type=::NewEnumerator1
      Normalize Vectors=false
     )
    )) {isSetParameters=true}
   /SetVariables(Particles.MeshOrientation=&MultiplyQuaternion(
     Quaternion A=$Particles.MeshOrientation
     Quaternion B=&MakeQuaternion(
      Quaternion From=2
      Axis=<0 0 0>
      Angle=0
      Y=&Multiply_Float(
       A=&Lerp_Float(
        A=&Multiply_Float(A=-0.75, B=$System.InterpolateOverTime.InterpolateOverTime_Float.Moving Average)
        B=0
        Alpha=$System.Dive
       )
       B=&MaskFloatBySpawnGroup(Masked Float=-1, Passthrough Float=1)
      )
      From=?
      To=?
      Angle Type=::NewEnumerator1
      Normalize Vectors=false
     )
    )) {isSetParameters=true}
   &ScaleSpriteSize(Uniform Curve Sprite Scale=DataInterface() {propertyValues {Curve {
       defaultValue=0
       keys [{interpMode=RCIM_Cubic, tangentMode=RCTM_User, time=0, value=0}, {interpMode=RCIM_Cubic, tangentMode=RCTM_User, time=1, value=1}]
      }}})
   /SetVariables(Particles.LightColor=&MultiplyLinearColor(
     Linear Color=<rgba 10 1.1419 5.5709 1>
     Scale Factor=&MakeVector4_FromFloat(Float=&FloatFromCurve(FloatCurve=DataInterface() {propertyValues {Curve {
          defaultValue=1
          keys [{interpMode=RCIM_Cubic, tangentMode=RCTM_User, time=0, value=0}, {interpMode=RCIM_Cubic, tangentMode=RCTM_User, time=1, value=1}]
         }}}))
    )) {isSetParameters=true}
   /SetVariables(Particles.Glow_Size:Vector2f=&Multiply_Vector2D_ByFloat(Float=2, Vector2D=$Particles.SpriteSize)) {isSetParameters=true}
   /SetVariables(Particles.Glow_Color:LinearColor=&MakeLinearColorFromVectorAndFloat(
     "Vector (RGB)"=&MakeVectorFromLinearColorRGB(Color=<rgba 2 0.04 1.02 1>)
     "Float (Alpha)"=&Multiply_Float(
      A=&FloatFromCurve(
       FloatCurve=DataInterface() {propertyValues {Curve {
          defaultValue=1
          keys [
           {interpMode=RCIM_Cubic, tangentMode=RCTM_User, time=0, value=0.4}
           {interpMode=RCIM_Cubic, tangentMode=RCTM_User, time=0.4, value=1}
           {tangentMode=RCTM_User, time=1, value=0.4}
          ]
         }}}
       CurveIndex=&ModuloFloat(A=&Multiply_Float(A=$Emitter.Age, B=0.2), B=1)
      )
      B=&FloatFromCurve(
       FloatCurve=DataInterface() {propertyValues {Curve {
          defaultValue=1
          keys [{interpMode=RCIM_Cubic, tangentMode=RCTM_User, time=0, value=0}, {interpMode=RCIM_Cubic, tangentMode=RCTM_User, time=1, value=1}]
         }}}
       Scale Curve=&FN_FrontEndFloatModifier(Gameplay=1, Front End=1)
      )
     )
    )) {isSetParameters=true}
   /SetVariables(Particles.GlowCamOffset=5) {isSetParameters=true}
   &OffsetPosition(Jitter Amount=0.5, Jitter Delay=0.05, Jitter Interpolation=1, Jitter Mode=1, Jitter Time Offset=0.1, Offset Mode=1, Position Output=1)
  ]
  rendererClasses [
   &NiagaraLightRendererProperties
   &NiagaraSpriteRendererProperties
   &NiagaraMeshRendererProperties
  ]
  renderers [
   {rendererClass=&NiagaraLightRendererProperties, rendererIndex=0}
   {rendererClass=&NiagaraSpriteRendererProperties, rendererIndex=1}
   {rendererClass=&NiagaraSpriteRendererProperties, rendererIndex=2}
   {rendererClass=&NiagaraMeshRendererProperties, rendererIndex=3}
  ]
 }
 Sparks_and_Stars {
  emitterUpdateScript [
   &EmitterState()
   &SpawnRate(
    SpawnRate=&Lerp_Float(
     A=100
     B=200
     Alpha=&FloatFromCurve(
      FloatCurve=DataInterface() {propertyValues {Curve {
         defaultValue=1
         keys [{interpMode=RCIM_Cubic, tangentMode=RCTM_User, time=0.2, value=0}, {interpMode=RCIM_Cubic, tangentMode=RCTM_User, time=1, value=1}]
        }}}
      CurveIndex=$System.Dive
     )
    )
    Use Spawn Probability=false
   )
  ]
  particleSpawnScript [
   &PartitionParticles()
   &InitializeParticle(
    Lifetime=&Multiply_Float(
     A=&RandomRangeFloat(Minimum=0.5, Maximum=0.6)
     B=&Lerp_Float(A=0.8, B=1, Alpha=$System.Dive)
    )
    Sprite Rotation Mode=::NewEnumerator2
    Sprite Size Max=<1 3>
    Sprite Size Min=<0.5 2>
    Sprite Size Mode=::NewEnumerator2
    UsePositionOffset=false
    Write Material Random=true
   )
   &CameraOffset(Camera Offset Amount=10)
   &SocketLocation(
    Interpolate Socket Sampling=true
    SamplePosOffset=false
    Socket Index=$Particles.PartitionParticles.Partition
    Socket Reader=$System.Socket Reader
    Socket Sampling Mode=::NewEnumerator3
    VelocitySpeedLimit=false
   )
   &ShapeLocation(
    Cylinder Height=&Lerp_Float(A=10, B=20, Alpha=$System.Dive)
    Cylinder Height Midpoint=0
    Cylinder Radius=10
    MaskSpawnGroup=false
    Override Seed=false
    Shape Primitive=::NewEnumerator1
    Surface Only=false
    UseOwnerScale=true
   )
   &AddVelocity(
    Cone Axis=&Multiply_VectorByFloat(
     Vector=&NormalizeVector(
      Vector To Normalize=&TransformVector(Vector=$System.SkydiveVelocity, Source Space=::World, Destination Space=::Local)
      Safe Normalize=true
      Safe Fallback Threshold=0
      Safe Fallback Vector=<0 0 0.001>
     )
     Float=-1
    )
    Module.UseConeFalloff=false
    Module.UseVelDistribution=false
    Override Seed=false
    Velocity Mode=::NewEnumerator2
    Velocity Speed=&Multiply_Float(
     A=&Lerp_Float(
      A=&RandomRangeFloat(Minimum=250, Maximum=300)
      B=&RandomRangeFloat(Minimum=400, Maximum=500)
      Alpha=$System.Dive
     )
     B=&SelectFloatFromArray(
      Direct Array Index=&Divide_Int(A:NiagaraInt32=$Particles.PartitionParticles.Partition, B:NiagaraInt32=2)
      Float Selection Array=DataInterface() {propertyValues {FloatData [1,0.5]}}
     )
    )
   )
   /SetVariables(Particles.SpawnGroup=&RandomRangeInt(Minimum=0, Maximum=8)) {isSetParameters=true}
   /SetVariables(Particles.VisibilityTag=&MaskIntBySpawnGroup(Masked Int=1, Passthrough Int=0, Spawn Group Mask=0)) {isSetParameters=true, module=SetVariables_#5}
  ]
  particleUpdateScript [
   &ParticleState()
   &SpriteRotationRate(Rotation Rate=&FloatFromCurve(
     FloatCurve=DataInterface() {propertyValues {Curve {
        defaultValue=1
        keys [
         {arriveTangent=-1.8015, interpMode=RCIM_Cubic, leaveTangent=-1.8015, tangentMode=RCTM_User, time=0, value=1}
         {interpMode=RCIM_Cubic, tangentMode=RCTM_User, time=1, value=0}
        ]
       }}}
     Scale Curve=&Multiply_Float(
      A=&RandomRangeFloat(Minimum=400, Maximum=600)
      B=&SelectFloatFromArray(
       Array Sampling Mode=::NewEnumerator0
       Direct Array Index=?
       Float Selection Array=DataInterface() {propertyValues {FloatData [-1,1]}}
       Random Seed=0
      )
     )
    ))
   &ScaleColor(
    Scale Alpha=&FloatFromCurve(FloatCurve=DataInterface() {propertyValues {Curve {
        defaultValue=1
        keys [{interpMode=RCIM_Cubic, tangentMode=RCTM_User, time=0, value=1}, {interpMode=RCIM_Cubic, tangentMode=RCTM_User, time=1, value=0}]
       }}})
    Scale RGB=&MakeVectorFromLinearColorRGB(Color=&ColorFromCurve(ColorCurve=DataInterface() {propertyValues {
        AlphaCurve {keys [<at 0 1 RCIM_Cubic>]}
        BlueCurve {keys [<at 0.0039 2.8539>, <at 0.2687 0.7439>, {interpMode=RCIM_Cubic, leaveTangent=-2.4783, time=0.6816, value=0.4868}, <at 1 0.5>]}
        GreenCurve {keys [<at 0.0039 0.6>, <at 0.2687 0.1919>, {interpMode=RCIM_Cubic, leaveTangent=-0.6317, time=0.6816, value=0}, <at 1 0>]}
        RedCurve {keys [<at 0.0039 3>, <at 0.2687 1>, {interpMode=RCIM_Cubic, leaveTangent=-2.8076, time=0.6816, value=1}, <at 1 0.3333>]}
       }}))
    ScaleA=true
    ScaleRGB=true
   )
   &ScaleSpriteSize(Scale Factor=&Vector2DFromFloat(Value=&FloatFromCurve(
      FloatCurve=DataInterface() {propertyValues {Curve {keys [
          {arriveTangent=10.1867, interpMode=RCIM_Cubic, leaveTangent=10.1867, tangentMode=RCTM_User, time=0, value=0}
          {interpMode=RCIM_Cubic, leaveTangent=-0.0263, tangentMode=RCTM_Break, time=0.2, value=1.2}
          {arriveTangent=-2.7631, interpMode=RCIM_Cubic, leaveTangent=-2.7631, tangentMode=RCTM_User, time=1, value=0}
         ]}}}
      Scale Curve=&FN_CameraDistance(
       0Dist:NiagaraFloat=50
       1Dist:NiagaraFloat=100
       Camera Query:NiagaraDataInterfaceCamera=DataInterface() {propertyValues {bRequireCurrentFrameData=false}}
       Position:NiagaraPosition=$Particles.Position
      )
     ))) {module=ScaleSpriteSize001}
   &CurlNoiseForce(Noise Frequency=100, Noise Strength=500, Pan Noise=true, Pan Noise Field=<0 0 0.5>, Random Seed=10)
   &Drag(Drag=&RandomRangeFloat(Minimum=2, Maximum=3), Use Linear Drag=true, Use Rotational Drag=false)
   &SolveForcesAndVelocity(Clamp Velocity=false, Limit Acceleration=false)
   &ScaleSpriteSizeBySpeed(Max Scale Factor=<0.75 2.5>, Min Scale Factor=<1 1.5>, Velocity Threshold=250)
   /SetVariables(Particles.Glow_Color:LinearColor=&MakeLinearColorFromVectorAndFloat("Vector (RGB)"=&MakeVectorFromLinearColorRGB(Color=$Particles.Color), "Float (Alpha)"=0.6)) {isSetParameters=true, module=SetVariables_#6}
   /SetVariables(Particles.Glow_Size:Vector2f=&Multiply_Vector2D(A=$Particles.SpriteSize, B=<4 2>)) {isSetParameters=true, module=SetVariables_#7}
   /SetVariables(Particles.StarSpriteSize:Vector2f=&Multiply_Vector2D_ByFloat(
     Float=&FloatFromCurve(
      FloatCurve=DataInterface() {propertyValues {Curve {
         defaultValue=1
         keys [
          {arriveTangent=5.2445, interpMode=RCIM_Cubic, leaveTangent=5.2445, tangentMode=RCTM_User, time=0, value=0}
          {interpMode=RCIM_Cubic, tangentMode=RCTM_User, time=0.5, value=1}
          {tangentMode=RCTM_User, time=1, value=0}
         ]
        }}}
      Scale Curve=&Multiply_Float(
       A=&RandomRangeFloat(Minimum=1.2, Maximum=1.5)
       B=&Lerp_Float(A=1, B=1.3, Alpha=$System.Dive)
      )
     )
     Vector2D=&Multiply_Vector2D_ByFloat(
      Float=&FN_CameraDistance(
       0Dist:NiagaraFloat=50
       1Dist:NiagaraFloat=70
       Camera Query:NiagaraDataInterfaceCamera=DataInterface() {propertyValues {bRequireCurrentFrameData=false}}
       Position:NiagaraPosition=$Particles.Position
      )
      Vector2D=<7 7>
     )
    )) {isSetParameters=true}
   /SetVariables(Particles.StarGlow_Size:Vector2f=&Multiply_Vector2D(A=$Particles.StarSpriteSize, B=<1.25 1.25>)) {isSetParameters=true}
   /SetVariables(Particles.Star Color:LinearColor=&ColorFromCurve(ColorCurve=DataInterface() {propertyValues {
       AlphaCurve {keys [<at 0.218 1 RCIM_Cubic>,<at 0.9987 0>]}
       BlueCurve {keys [{arriveTangent=-0.3335, interpMode=RCIM_Cubic, time=0.0752, value=0.57}, <at 0.5066 0.6>, <at 1 0.4>]}
       GreenCurve {keys [{arriveTangent=-0.2624, interpMode=RCIM_Cubic, time=0.0752, value=0.12}, <at 0.5066 0>, <at 1 0>]}
       RedCurve {keys [{arriveTangent=-0.9116, interpMode=RCIM_Cubic, time=0.0752, value=1.2}, <at 0.5066 0.8>, <at 1 0.2333>]}
      }})) {isSetParameters=true}
   /SetVariables(Particles.Star Glow Color:LinearColor=&MultiplyLinearColor(Linear Color=$Particles.Star Color, Scale Factor=<1.1 1.1 1.1 0.8>)) {isSetParameters=true}
  ]
  rendererClasses [&NiagaraSpriteRendererProperties]
  renderers [
   {rendererClass=&NiagaraSpriteRendererProperties}
   {rendererClass=&NiagaraSpriteRendererProperties, rendererIndex=1}
   {rendererClass=&NiagaraSpriteRendererProperties, rendererIndex=2}
   {rendererClass=&NiagaraSpriteRendererProperties, rendererIndex=3}
  ]
 }
}
summary {
 systemName=NS_Contrail_ForestJean
 emitters [
  {emitterName=Flames_Sprites, rendererClasses [&NiagaraSpriteRendererProperties]}
  {
   emitterName=Light_and_Glow_and_Horse_Shoe
   rendererClasses [
    &NiagaraLightRendererProperties
    &NiagaraSpriteRendererProperties
    &NiagaraMeshRendererProperties
   ]
  }
  {emitterName=Sparks_and_Stars, rendererClasses [&NiagaraSpriteRendererProperties]}
  {emitterName=Bokeh_and_Sparkles, rendererClasses [&NiagaraSpriteRendererProperties]}
  {emitterName=Flame_Ribbon, rendererClasses [&NiagaraRibbonRendererProperties]}
 ]
 userVariables [
  User.bIsFrontEnd {type=NiagaraBool}
  User.bIsFrontEndPreview {type=NiagaraBool}
  User.ParaGlideLeanAlpha {type=NiagaraFloat, defaultValue=0, description=""}
 ]
}
